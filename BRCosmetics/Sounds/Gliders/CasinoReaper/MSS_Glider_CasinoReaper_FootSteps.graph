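# MetaSound Source preset that plays footstep sounds for the Casino Reaper glider cosmetic as a stereo one-shot.
# It inherits its logic from the shared MSS_Glider_Open parent and overrides inputs, loading ten footstep waves into the Open.Sounds array.
# Narrow pitch and volume ranges add variation around a slightly raised pitch and reduced volume, with no start delay.
# It reuses glider open/close attenuation and concurrency, routed to the glider land sound class.
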MetaSoundSource /BRCosmetics/Sounds/Gliders/CasinoReaper/MSS_Glider_CasinoReaper_FootSteps
preset_parent=/FortMetasound/Cosmetics/Glider/MSS_Glider_Open
inputs [
 Open.Sounds {type=WaveAsset:Array}
 PitchModulationRange.Max {type=Float}
 PitchModulationRange.Min {type=Float}
 StartDelay.Max {type=Time}
 StartDelay.Min {type=Time}
 TemplatePitchMultiplier {type=Float}
 TemplateVolumeMultiplier {type=Float}
 UE.Source.OnPlay {type=Trigger}
 VolumeModulationRange.Max {type=Float}
 VolumeModulationRange.Min {type=Float}
]
interfaces [UE.OutputFormat.Stereo:1.0, UE.Source.OneShot:1.0, UE.Source:1.1]
outputs [
 "UE.OutputFormat.Stereo.Audio:0" {type=Audio}
 "UE.OutputFormat.Stereo.Audio:1" {type=Audio}
 UE.Source.OneShot.OnFinished {type=Trigger}
]
pages [Default {input_defaults {
   PitchModulationRange.Max=1.02
   PitchModulationRange.Min=0.98
   StartDelay.Max=0
   StartDelay.Min=0
   TemplatePitchMultiplier=1.05
   TemplateVolumeMultiplier=0.4
   UE.Source.OnPlay=null
   VolumeModulationRange.Max=1.05
   VolumeModulationRange.Min=0.98
   Open.Sounds [
    &/Glider_CasinoReaper_FootStep_01
    &/Glider_CasinoReaper_FootStep_02
    &/Glider_CasinoReaper_FootStep_03
    &/Glider_CasinoReaper_FootStep_04
    &/Glider_CasinoReaper_FootStep_05
    &/Glider_CasinoReaper_FootStep_06
    &/Glider_CasinoReaper_FootStep_07
    &/Glider_CasinoReaper_FootStep_08
    &/Glider_CasinoReaper_FootStep_09
    &/Glider_CasinoReaper_FootStep_10
   ]
  }}]
properties {
 assetImportData=&self:AssetImportData
 attenuationSettings=/Game/Sounds/Attenuation/Fort_Player_Glider_OpenClose_Athena_Att
 maxDistance=10000
 numChannels=2
 outputFormat=Stereo
 sampleRate=48000
 soundClassObject=&Fort_Gliders_Land
 concurrencySet [&Athena_GliderOpenClose_Conc]
}
